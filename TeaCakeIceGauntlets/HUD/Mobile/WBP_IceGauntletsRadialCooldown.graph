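# WBP_IceGauntletsRadialCooldown is a mobile HUD widget Blueprint for the TeaCake Ice Gauntlets, derived from the Athena Bacchus WBP_RadialButton.
# It adds no widgets or variables of its own.
# PreConstruct and Construct only call their parent implementations, and Tick is an empty event, so the file itself defines no cooldown logic.
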WidgetBlueprint /TeaCakeIceGauntlets/HUD/Mobile/WBP_IceGauntletsRadialCooldown
graphs [EventGraph {nodes [
   n1 = Event PreConstruct()
   > Parent: PreConstruct(IsDesignTime=@n1.IsDesignTime)
   Event Construct()
   > Parent: Construct()
   Event Tick()
  ]}]
variables []
widgets {
 info {inheritedWidgetCount=0, parentClass=/Game/Athena/HUD/Bacchus/WBP_RadialButton.WBP_RadialButton_C, rootWidgetClass=null, widgetCount=0}
 widgets []
}
